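# Physics asset for the large male Choco Bunny cosmetic head, providing physics bodies and joints for its two ears.
# Each ear chains a base, lower and upper capsule body back to a small head body.
# The bases swing and twist freely on the head, while the outer joints carry narrower angular limits, with the right ear allowing wider secondary swing than the left.
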PhysicsAsset /BRCosmetics/Characters/Player/Male/Large/Heads/M_LRG_Choco_Bunny_Head_01/Meshes/Male_Large_Choco_Bunny_Head_Physics
bodies {
 dyn_ear_a_base_l [{shapeType=Capsule, length=1, radius=1}]
 dyn_ear_a_base_r [{shapeType=Capsule, length=1, radius=1}]
 dyn_ear_a_l [{shapeType=Capsule, length=2.75, radius=4, center=<-2.3847 0 0.0001>, rotation=<rot 90 0 0>}]
 dyn_ear_a_r [{shapeType=Capsule, length=2.75, radius=4, center=<2.3847 0 -0.0001>, rotation=<rot 90 0 180>}]
 dyn_ear_b_l [{shapeType=Capsule, length=4.25, radius=3.75, center=<-3.303 0 -0.0001>, rotation=<rot 90 0 0>}]
 dyn_ear_b_r [{shapeType=Capsule, length=4.25, radius=3.75, center=<3.303 0 0.0001>, rotation=<rot 90 0 180>}]
 dyn_head [{shapeType=Capsule, length=1, radius=1}]
}
constraints [
 {
  bone1Name=dyn_ear_a_base_r
  bone2Name=dyn_head
  swing1LimitDegrees=45
  swing1Motion=Free
  swing2LimitDegrees=45
  swing2Motion=Free
  twistLimitDegrees=45
  twistMotion=Free
 }
 {
  bone1Name=dyn_ear_a_base_l
  bone2Name=dyn_head
  swing1LimitDegrees=45
  swing1Motion=Free
  swing2LimitDegrees=45
  swing2Motion=Free
  twistLimitDegrees=45
  twistMotion=Free
 }
 {bone1Name=dyn_ear_a_l, bone2Name=dyn_ear_a_base_l, swing1LimitDegrees=5, swing2LimitDegrees=5, twistLimitDegrees=5}
 {bone1Name=dyn_ear_b_l, bone2Name=dyn_ear_a_l, swing1LimitDegrees=10, swing2LimitDegrees=30, twistLimitDegrees=5}
 {bone1Name=dyn_ear_b_r, bone2Name=dyn_ear_a_r, swing1LimitDegrees=10, swing2LimitDegrees=45, twistLimitDegrees=5}
 {bone1Name=dyn_ear_a_r, bone2Name=dyn_ear_a_base_r, swing1LimitDegrees=5, swing2LimitDegrees=10, twistLimitDegrees=5}
]
