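# Animation Blueprint for the large male Choco Bunny head cosmetic, adding secondary physics and an emote-driven mouth change.
# The AnimGraph copies pose and curves from the attached parent mesh, then runs a RigidBody simulation in neck_01 space using the head's physics asset, with damped component motion and overridden gravity.
# Each animation update switches the "Smile Mouth animation Frame" dynamic material parameter according to whether an emote is playing.
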AnimBlueprint /BRCosmetics/Characters/Player/Male/Large/Heads/M_LRG_Choco_Bunny_Head_01/Meshes/Male_Large_Choco_Bunny_Head_AnimBP
target_skeleton=&/Male_Large_Choco_Bunny_Head_Skeleton
graphs [
 AnimGraph {nodes [
   n1 = RigidBody(ComponentPose=Local To Component(LocalPose=Copy Pose From Mesh() {bCopyCurves=true, bUseAttachedParent=true})) {
    bClampLinearTranslationLimitToRefPose=true
    bOverrideWorldGravity=true
    overridePhysicsAsset=&/Male_Large_Choco_Bunny_Head_Physics
    simulationSpace=BaseBoneSpace
    alphaBoolBlend {blendOption=Linear}
    baseBoneRef {boneName=neck_01}
    componentLinearAccScale=<0.3 0.3 0.3>
    componentLinearVelScale=<0.3 0.3 0.3>
    overrideWorldGravity {z=-980}
   }
   Output Pose(Result=Component To Local(ComponentPose=@n1)) {layerGroup=None}
  ]}
 EventGraph {nodes [
   Event BlueprintUpdateAnimation()
   > n1 = Get Dynamic Material Instance(SkeletalMeshComponent=GetPartSkeletalMeshComponent())
   > n2 = Branch(Condition=Get bIsPlayingEmote())
   > SetScalarParameterValue(self=@n1, ParameterName=Smile Mouth animation Frame, Value=8)
   SetScalarParameterValue(execute=@n2.else, self=@n1, ParameterName=Smile Mouth animation Frame, Value=3)
   TryGetPawnOwner()
  ]}
]
variables []
